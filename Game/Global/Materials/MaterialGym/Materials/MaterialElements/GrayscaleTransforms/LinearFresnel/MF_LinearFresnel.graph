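# Material function that outputs a grayscale Fresnel mask varying linearly with the angle between camera and normal vectors, both overridable inputs.
# A fast arccosine of their normalized dot product is divided by half pi and saturated, giving 0 facing the camera and 1 at grazing angles.
# That value is remapped between Min and Max by inverse lerp and saturated.
# An If comparing Min and Max effectively returns Max in every branch.
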MaterialFunction /Game/Global/Materials/MaterialGym/Materials/MaterialElements/GrayscaleTransforms/LinearFresnel/MF_LinearFresnel
graph {nodes [
  n1 = FunctionInput() {bUsePreviewValueAsDefault=true, inputName=Min, inputType=FunctionInput_Scalar}
  n2 = FunctionInput() {bUsePreviewValueAsDefault=true, inputName=Max, inputType=FunctionInput_Scalar, sortPriority=1, previewValue=<1 0 0 1>}
  n3 = If(A=@n1, B=@n2, "A > B"=@n2, "A == B"=Add(A=@n2) {constB=0}, "A < B"=@n2)
  n4 = CameraVectorWS()
  n5 = PixelNormalWS()
  n6 = DotProduct(
   A=Normalize(VectorInput=FunctionInput(Preview=@n4) {bUsePreviewValueAsDefault=true, inputName=Camera Vector, sortPriority=2})
   B=Normalize(VectorInput=FunctionInput(Preview=@n5) {bUsePreviewValueAsDefault=true, inputName=Normal Vector, sortPriority=3})
  )
  n7 = Constant() {r=0.5}
  n8 = Saturate(None=Divide(
    A=ArccosineFast!(None=@n6)
    B=MaterialFunctionCall(Multiplier=@n7) {materialFunction=&Pi}.Result
   ))
  FunctionOutput(None=Saturate(None=InverseLinearInterpolate(A=@n1, B=@n3, Value=@n8)))
 ]}
