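# M_LinearFresnel is a Material Gym material, flagged for skeletal meshes, that demonstrates a linear Fresnel gradient eroded by animated noise.
# MF_LinearFresnel, bounded by Min and Max, feeds MF_Erode alongside wave-flame noise scrolling on camera-aligned virtual-plane UVs.
# The eroded mask blends MF_SimpleSlurp's result toward a B-scaled copy of itself.
# Preskinned gradient, triplanar box, light vector and alternative UV nodes sit outside that chain.
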Material /Game/Global/Materials/MaterialGym/Materials/MaterialElements/GrayscaleTransforms/M_LinearFresnel
graph {
 properties {bAutomaticallySetUsageInEditor=false, bUsedWithMeshDeformer=true, bUsedWithSkeletalMesh=true, usedShadingModels=Default Lit}
 nodes [
  AtmosphericLightVector()
  n1 = ScalarParameter() {defaultValue=0.5271, parameterName=Max}
  n2 = ScalarParameter() {defaultValue=0.1141, parameterName=Min}
  ComponentMask(None=MaterialFunctionCall!(Min=@n1, Max=@n2, Pow) {materialFunction=/Game/Global/Materials/MaterialGym/Materials/MaterialElements/GrayscaleTransforms/BoundsGradient/MF_PreskinnedGradient}."Gradient (V3)") {b=true}
  n3 = MaterialFunctionCall() {materialFunction=/Game/Global/Materials/MaterialGym/Materials/MaterialElements/Depth/SimpleSlurp/MF_SimpleSlurp}
  n4 = ScalarParameter() {defaultValue=5.6657, parameterName=Tile}
  n5 = Multiply(A=Time(), B=Constant2Vector() {g=0.15}) {constB=12}
  n6 = MaterialFunctionCall!(UV Tiling=@n4, UV Offset=@n5) {materialFunction=/Game/Global/Materials/MaterialGym/Materials/MaterialElements/UVs/CameraAlignedVirutalPlane/MF_CameraAlignedVirtualPlane}
  LinearInterpolate(
   A=@n3.Result
   B=Multiply(A=@n3.Result, B=ScalarParameter() {defaultValue=1.5517, parameterName=B})
   Alpha=MaterialFunctionCall!(
    Gradient=MaterialFunctionCall!(Min=@n2, Max=@n1, Camera Vector, Normal Vector) {materialFunction=/Game/Global/Materials/MaterialGym/Materials/MaterialElements/GrayscaleTransforms/LinearFresnel/MF_LinearFresnel}.Result
    Noise=TextureSample(UVs=@n6.OutputUV) {texture=/Game/Effects/Fort_Effects/Textures/Tile/T_Tile_WaveFlame}.R
    Contrast=ScalarParameter() {defaultValue=2.0964, parameterName=ErodeContrast}
   ) {materialFunction=/Game/Global/Materials/MaterialGym/Materials/MaterialElements/GrayscaleTransforms/Erode/MF_Erode}.Result
  )
  MaterialFunctionCall!(
   TVB_Texture=TextureObject() {texture=&T_Tile_WaveFlame}
   TVB_UV_Tiling=@n4
   TVB_UV_Offset
   TVB_CenterPointOffset
   TVB_Depth
   TVB_BlendContrast
   TVB_UseLocalCameraVector
  ) {materialFunction=/Game/Global/Materials/MaterialGym/Materials/MaterialElements/UVs/PreskinnedVirtualBox/MF_TriplanarVirtualBox}
  Multiply(A=@n4, B=Add(A=TextureCoordinate(), B=@n5))
 ]
}
